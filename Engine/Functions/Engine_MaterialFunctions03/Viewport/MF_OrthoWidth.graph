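# Material function exposed to the library that returns the orthographic view's width, or zero when the view is not orthographic.
# A Custom HLSL node multiplies the IsOrthographic flag by `View.ClipToView[0][0]` and a scale factor.
# A static bool input, Half Width, defaults to true and selects between half width (factor 1) and full width (factor 2).
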MaterialFunction /Engine/Functions/Engine_MaterialFunctions03/Viewport/MF_OrthoWidth
graph {
 properties {bExposeToLibrary=true}
 nodes [
  MaterialExpressionIsOrthographic_0 = IsOrthographic()
  MaterialExpressionStaticSwitch_0 = StaticSwitch(
   True=Constant() {r=1}
   False=Constant() {r=2}
   Value=FunctionInput(Preview=StaticBoolParameter() {defaultValue=true, parameterName=Default Half Width}) {bUsePreviewValueAsDefault=true, inputName=Half Width, inputType=FunctionInput_StaticBool}
  )
  n1 = Custom!(IsOrthoFloat=@MaterialExpressionIsOrthographic_0, Half=@MaterialExpressionStaticSwitch_0) {
   code="return IsOrthoFloat * View.ClipToView[0][0] * Half;"
   outputType=CMOT_Float1
   showCode=true
   inputs [
    {
     inputName=IsOrthoFloat
     input {expression=&self:MaterialExpressionIsOrthographic_0}
    }
    {inputName=Half, input {expression=&self:MaterialExpressionStaticSwitch_0}}
   ]
  }
  FunctionOutput(None=@n1)
 ]
}
